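# Looping ambient cue for an individual wind turbine spot at the Helios POI, in the Athena ambient props sound class.
# One looping mid-layer turbine wave feeds an Attenuation node with its own falloff override, then a default Modulator at the root.
# The cue also stores wider, non-spatialized capsule attenuation with occlusion, but its override flag is not set, so the node's tighter settings define the audible range.
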SoundCue /Game/Sounds/Helios/POI/WindTurbines/Spots/Spot_Helios_Turbine_Individual_Cue
root=Modulator_0
nodes [
 {
  id=Attenuation_1
  children [WavePlayer_0]
  properties {bOverrideAttenuation=true, attenuationOverrides {falloffDistance=3000, attenuationShapeExtents=<1500 3000>}}
 }
 {id=Modulator_0, children [Attenuation_1]}
 {
  id=WavePlayer_0
  properties {bLooping=true, soundWaveAssetPtr=&/Spot_Helios_Wind_Turbines_01_Loop_Mid}
 }
]
properties {
 maxDistance=4500
 soundClassObject=&Fort_Ambient_Props_Athena
 volumeMultiplier=0.702
 attenuationOverrides {
  attenuationShape=Capsule
  bEnableOcclusion=true
  bSpatialize=false
  falloffDistance=8000
  nonSpatializedRadiusEnd=10000
  nonSpatializedRadiusStart=5000
  occlusionInterpolationTime=2
  occlusionLowPassFilterFrequency=600
  occlusionTraceChannel=ECC_GameTraceChannel11
  reverbWetLevelMin=0.6905
  stereoSpread=180
  attenuationShapeExtents=<8000 3000 800>
 }
}
