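# Sound cue for a looping stereo wind turbine wash, a spot ambience for the Helios turbine POI in the ambient props sound class.
# A single looping wave player feeds an Attenuation node that overrides falloff with a large, non-spatialized capsule using a logarithmic curve.
# The cue also stores its own capsule settings with occlusion filtering, but its attenuation override flag is not set.
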SoundCue /Game/Sounds/Helios/POI/WindTurbines/Spots/Spot_Helios_Turbine_Wash_Loop_Stereo_Cue
root=Attenuation_1
nodes [
 {
  id=Attenuation_1
  children [WavePlayer_4]
  properties {
   bOverrideAttenuation=true
   attenuationOverrides {
    attenuationShape=Capsule
    bSpatialize=false
    distanceAlgorithm=Logarithmic
    falloffDistance=8000
    focusAzimuth=90
    focusVolumeAttenuation=114.3622
    nonFocusAzimuth=270
    attenuationShapeExtents=<11000 4000>
   }
  }
 }
 {
  id=WavePlayer_4
  properties {bLooping=true, soundWaveAssetPtr=&/Spot_Helios_Wind_Turbines_01_Loop_Side}
 }
]
properties {
 maxDistance=19000
 soundClassObject=&Fort_Ambient_Props_Athena
 volumeMultiplier=0.902
 attenuationOverrides {
  attenuationShape=Capsule
  bEnableOcclusion=true
  bSpatialize=false
  falloffDistance=8000
  nonSpatializedRadiusEnd=10000
  nonSpatializedRadiusStart=5000
  occlusionInterpolationTime=2
  occlusionLowPassFilterFrequency=600
  occlusionTraceChannel=ECC_GameTraceChannel11
  reverbWetLevelMin=0.6905
  stereoSpread=180
  attenuationShapeExtents=<8000 3000 800>
 }
}
